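; BP_Sneezy_CountComponent is a Blueprint component in the Sneezy prototype that relays a dancer count to listeners.
; Its custom event ServerCountDancers takes an integer and broadcasts it through the OnCountDancers event dispatcher.
; BeginPlay and Tick are empty, so the component does no counting itself.
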(event Custom|ServerCountDancers
  (params (CountDancers :type Integer))
  (Default|CallOnCountDancers CountDancers))

(event EventBeginPlay)

(event EventTick
  (params (DeltaSeconds :type Float%20%28single-precision%29)))
