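; Dance floor event prop that turns players' emotes into progress on a global sound meter, ending in a celebration.
; Authority polls its box on a timer for emoting pawns and broadcasts dancer-count changes, while an externally supplied MeterLevel returns the global total.
; RepNotifies drive timelines for the meter needle, magnetic spline materials and filtered club music.
; A full meter plays the celebration, teleports the pedestal away and disables the floor.
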(event Custom|SetMagneticMIDsWorldPosition
  (Networking|CanExecuteCosmeticEvents
    (:True
      (for _array_element (Variables|Default|GetMagneticSplineMIDs)
        (Rendering|Material|SetVectorParameterValue _array_element "DanceFloorCenterPoint" (Math|Conversions|ToLinearColor(Vector) (Transformation|GetWorldLocation (Variables|Default|GetMagneticSceneComponent))) :!class /Script/Engine.MaterialInstanceDynamic)))))

(event EventBeginPlay
  ;[1 1 1] Temp - please replace
  (CallFunction|GlobalNeedleMove 0.0)
  (CallFunction|SetMagneticMIDsWorldPosition)
  (Class|FortAsyncActionAuthorityMatchReady|MatchReadyAsync))

(event Custom|SetDancefloorEnabled
  (params (Enabled :type Boolean))
  ;[1 1 1] set both to true for testing!
  ;;[5 3.613 0.1215] RECONNECT
  (if Enabled
    (Collision|SetActorEnableCollision true)
    (Rendering|SetActorHiddenInGame false)
    (Components|Activation|SetActive (Variables|Default|GetP_Sneezy_Rays) true true)
    (Variables|Default|SetPlayDancefloorMusic true)
    (else
      (Collision|SetActorEnableCollision false)
      (Rendering|SetActorHiddenInGame true)
      (Variables|Default|SetPlayDancefloorMusic false))))

(event Collision|EventActorBeginOverlap
  (params (OtherActor :type Actor%20Object%20Reference)))

(event OnComponentBeginOverlap(Box)
  (params (OverlappedComponent :type Primitive%20Component%20Object%20Reference) (OtherActor :type Actor%20Object%20Reference) (OtherComp :type Primitive%20Component%20Object%20Reference) (OtherBodyIndex :type Integer) (bFromSweep :type Boolean) (SweepResult :type Hit%20Result%20Structure%20%28by%20ref%29))
  ;[1 1 1] Overlap to start check for dancers
  (Networking|SwitchHasAuthority
    (:Authority
      (Utilities|IsValid (Fortnite|GetFortPlayerControllerfromActor OtherActor)
        (:Is%20Valid
          (if (not (Utilities|Time|IsValidTimerHandle (Variables|Default|GetDanceCheckTimer)))
            (bind _returnvalue (Utilities|Time|SetTimerbyEvent (EventDispatchers|CreateEvent :!event Custom|CheckForDancers) (Variables|Default|GetBeatTiming) true))
            (Variables|Default|SetDanceCheckTimer _returnvalue)))))))

(event Custom|CheckForDancers
  (bind _box (Variables|Default|GetBox))
  (bind _overlappingactors (Collision|GetOverlappingActors _box "/Script/FortniteGame.FortPlayerPawnAthena" :!class /Script/Engine.PrimitiveComponent))
  ;[1 1 1] Check for dancers
  (Variables|Default|SetDancerCount 0)
  (for _array_element _overlappingactors
    (if (Emote|IsPlayingEmote (Fortnite|GetFortPlayerControllerfromActor _array_element))
      (Math|Integer|IncrementInt (Variables|Default|GetDancerCount))))
  (Variables|Default|SetLocalEdgeDancers (- (Variables|Default|GetDancerCount) (Variables|Default|GetPreviousDancerCount)))
  (Variables|Default|SetPreviousDancerCount (Variables|Default|GetDancerCount))
  (if (< (Utilities|Array|Length _overlappingactors) 1)
    (Utilities|Time|ClearandInvalidateTimerbyHandle (Variables|Default|GetDanceCheckTimer))))

(event Custom|EdgeServerMeterLevel
  (params (MeterLevel :type Float%20%28double-precision%29))
  ;[1 1 1] Check for dancers
  (Variables|Default|SetMeterLevel MeterLevel))

(event Custom|DebugSoundLevel
  (params (NumOfDancers :type Float%20%28double-precision%29))
  ;[1 1 1] Debug visuals for world wide dancer % of threshold
  (Variables|Default|SetMeterLevel NumOfDancers))

(event Custom|GlobalNeedleMove
  (params (SoundMeterLevel :type Float%20%28double-precision%29))
  (bind _meterlevel (Variables|Default|GetMeterLevel))
  (bind _soundmeterstartinglevel (Variables|Default|GetSoundMeterStartingLevel))
  (bind _returnvalue (- _soundmeterstartinglevel _meterlevel))
  (bind _returnvalue_2 (Math|Float|Absolute(Float) _returnvalue))
  (bind _thresholdtoshowmagneticsplines (Variables|Default|GetThresholdToShowMagneticSplines))
  (bind _meterlevel_2 (Variables|Default|GetMeterLevel))
  (bind _returnvalue_3 (>= _meterlevel_2 _thresholdtoshowmagneticsplines))
  (bind _meterlevel_3 (Variables|Default|GetMeterLevel))
  (bind _returnvalue_4 (Math|Float|MapRangeClamped _meterlevel_3 0.3 0.8 0.1 1.0))
  (bind _soundleveltimemultiplier (Variables|Default|GetSoundLevelTimeMultiplier))
  (bind _returnvalue_5 (* _returnvalue_2 _soundleveltimemultiplier))
  (bind _meterlevel_4 (Variables|Default|GetMeterLevel))
  (bind _soundmeterstartinglevel_2 (Variables|Default|GetSoundMeterStartingLevel))
  ;[1 1 1] FX / Sound Meter
  (Networking|SwitchHasAuthority
    (:Authority
      _)
    (:Remote
      (bind _soundmetermid (Variables|Default|GetSoundMeterMID)
        (:then
          (bind _returnvalue_6 (Rendering|Material|GetScalarParameterValue _soundmetermid "NormalizedLevel" :!class /Script/Engine.MaterialInstanceDynamic))
          (Variables|Default|SetSoundMeterStartingLevel _returnvalue_6))
        (:else
          _))
      (bind _returnvalue_7 (Rendering|Material|GetScalarParameterValue (Utilities|Array|Get(acopy) (Variables|Default|GetMagneticSplineMIDs)) "Opacity" :!class /Script/Engine.MaterialInstanceDynamic))
      (Variables|Default|SetMagneticOpacityStartingLevel _returnvalue_7)
      (for _array_element (Variables|Default|GetMagneticSplines)
        (Development|SetHiddeninGame _array_element (not _returnvalue_3) false))))
  (Components|Activation|SetActive (Variables|Default|GetCoalesce) _returnvalue_3 false)
  (if (>= (Variables|Default|GetMeterLevel) 1.0)
    (CallFunction|PlayCelebration)
    (else
      _))
  (if (not (Math|Float|NearlyEqual(Float) _returnvalue_5 0.0 0.01))
    (Components|Timeline|SetPlayRate (Variables|BP_Sneezy_FloorProp_Area|GetUpdateSoundLevel) (Utilities|Reciprocal _returnvalue_5))
    (bind (_direction _0_1) (Timeline|UpdateSoundLevel PlayFromStart)
      (:Update
        (Rendering|Material|SetScalarParameterValue (Variables|Default|GetSoundMeterMID) "NormalizedLevel" (Math|Float|Lerp _soundmeterstartinglevel_2 _meterlevel_4 _0_1) :!class /Script/Engine.MaterialInstanceDynamic)
        (CallFunction|UpdateMeterLevelAudio (Math|Float|Lerp _soundmeterstartinglevel_2 _meterlevel_4 _0_1)) ; This scales intensity of a buncha audio
        (Development|PrintString (Utilities|String|ToString(Float) (Math|Float|Lerp _soundmeterstartinglevel_2 _meterlevel_4 _0_1)))
        (Networking|ForceNetUpdate)
        (Networking|SwitchHasAuthority
          (:Remote
            (Rendering|Material|SetScalarParameterValue (Variables|Default|GetFFT_DanceFloorMID) "FFTOpacity" _returnvalue_4 :!class /Script/Engine.MaterialInstanceDynamic)
            (for _array_element_2 (Variables|Default|GetMagneticSplineMIDs)
              (Rendering|Material|SetScalarParameterValue _array_element_2 "FFTOpacity" _returnvalue_4 :!class /Script/Engine.MaterialInstanceDynamic)
              (Rendering|Material|SetScalarParameterValue _array_element_2 "Opacity" (Math|Float|Lerp (Variables|Default|GetMagneticOpacityStartingLevel) (Math|Float|MapRangeClamped (Variables|Default|GetMeterLevel) 0.5 0.75 0.1 1.0) _0_1) :!class /Script/Engine.MaterialInstanceDynamic)))))
      (:Finished
        _)))
  (CallFunction|LocalNeedleUp))

(event Custom|PlayCelebration
  (bind _gatewayhatch_pedestal_pedestal02 (Variables|Default|GetGatewayHatch_Pedestal_pedestal02))
  (bind (_returnvalue_location _returnvalue_rotation _returnvalue_scale) (Math|Transform|BreakTransform (Transformation|GetWorldTransform _gatewayhatch_pedestal_pedestal02)))
  (bind _gatewayhatch_pedestal_pedestal02_2 (Variables|Default|GetGatewayHatch_Pedestal_pedestal02))
  ;[1 1 1] FX / Sound Meter
  (Collision|SetCollisionEnabled (Variables|Default|GetBox) "NoCollision")
  (Utilities|FlowControl|Sequence
    (:then_0
      (Components|Activation|SetActive (Variables|Default|GetCelebration) true false)
      (Audio|PlaySoundatLocation (Variables|Sound|GetSoundOnCelebration) (Transformation|GetWorldLocation (Variables|Default|GetAmb)))
      (Audio|Components|Audio|FadeOut (Variables|Default|GetMusic) 0.5 0.0 "Linear" :!class /Script/Engine.AudioComponent)
      (bind (_direction _sweep _tipscale _opacity) (Timeline|CelebrationTL PlayFromStart)
        (:Update
          (for _array_element (Variables|Default|GetMagneticSplineMIDs)
            (Rendering|Material|SetScalarParameterValue _array_element "TipCenter" _sweep :!class /Script/Engine.MaterialInstanceDynamic)
            (Rendering|Material|SetScalarParameterValue _array_element "TipScale" _tipscale :!class /Script/Engine.MaterialInstanceDynamic)
            (Rendering|Material|SetScalarParameterValue _array_element "Opacity" _opacity :!class /Script/Engine.MaterialInstanceDynamic))
          (Rendering|Material|SetScalarParameterValue (Variables|Default|GetFFT_DanceFloorMID) "FFTOpacity" _opacity :!class /Script/Engine.MaterialInstanceDynamic))
        (:Finished
          (Components|Activation|SetActive (Variables|Default|GetCoalesce) false false)
          (Development|SetHiddeninGame (Variables|Default|GetFFT_DanceFloor) true false)
          (for _array_element_2 (Variables|Default|GetMagneticSplines)
            (Development|SetHiddeninGame _array_element_2 true false)))))
    (:then_1
      (Development|SetHiddeninGame (Variables|Default|GetSoundMeter) true false)
      (bind _output_get (Variables|Default|SetStartLocation (Transformation|GetWorldLocation _gatewayhatch_pedestal_pedestal02_2)))
      (bind _output_get_2 (Variables|Default|SetStartRotation (Transformation|GetWorldRotation _gatewayhatch_pedestal_pedestal02_2)))
      (Variables|Default|SetEndLocation (+ _output_get "0.000000,-1000.000000, 0"))
      (Variables|Default|SetEndRotation (Math|Rotator|CombineRotators _output_get_2 "0.000000,-90.000000,-55.000000"))
      (bind (_direction_2 _newtrack_0) (Timeline|PedestalMove PlayFromStart)
        (:Update
          (Transformation|SetWorldLocationAndRotation (Variables|Default|GetGatewayHatch_Pedestal_pedestal02) (Math|Vector|Lerp(Vector) (Variables|Default|GetStartLocation) (Variables|Default|GetEndLocation) _newtrack_0) (Math|Rotator|Lerp(Rotator) (Variables|Default|GetStartRotation) (Variables|Default|GetEndRotation) _newtrack_0) false false))
        (:Finished
          (CallFunction|SetDancefloorEnabled false)
          (Effects|SpawnEmitteratLocation "/Figment_S08/Athena/Prototype/Blueprints/White/FX/P_Snow_Pin_TeleportOut.P_Snow_Pin_TeleportOut" _returnvalue_location _returnvalue_rotation _returnvalue_scale)
          (Camera|PlayWorldCameraShake "/Game/Effects/Fort_Effects/BluePrints/CameraShakes/B_Small_Vertical_Jolt_CameraShake.B_Small_Vertical_Jolt_CameraShake_C" _returnvalue_location 3e+03 7e+03))
        (:PowerUp
          (Effects|SpawnEmitterAttached :EmitterTemplate "/Figment_S08/Athena/Prototype/Blueprints/White/FX/P_Snow_Pin_CompletionBuildup.P_Snow_Pin_CompletionBuildup" :AttachToComponent _gatewayhatch_pedestal_pedestal02 :AttachPointName "None" :Scale "1.000000,1.000000,1.000000" :LocationType "SnapToTargetIncludingScale" :bAutoDestroy true :PoolingMethod "None" :bAutoActivate true)
          (Audio|SpawnSoundAttached "/Figment_S08/Athena/Sounds/Events/Dopey/Snow/SC_Snow_ChallComp_TeleportOut_Cue.SC_Snow_ChallComp_TeleportOut_Cue" (Variables|Default|GetGatewayHatch_Pedestal_pedestal02)))))))

(event Custom|Open%20Music%20Filter
  (params (Time :type Float%20%28double-precision%29))
  ;[1 1 1] Start / Stop Club Music
  (Audio|Components|Audio|SetLowPassFilterEnabled (Variables|Default|GetMusic) true :!class /Script/Engine.AudioComponent)
  (Components|Timeline|SetPlayRate (Variables|BP_Sneezy_FloorProp_Area|GetClub%20Music%20Fader) (/ 1.0 Time))
  (if (Variables|Sound|GetMusicIsFiltered)
    (Audio|PlaySoundatLocation (Variables|Sound|GetSoundOnMusicStart) (Transformation|GetWorldLocation (Variables|Default|GetMusic)))
    (Variables|Sound|SetMusicIsFiltered false)
    (bind (_direction _crossfade) (Timeline|Club%20Music%20Fader PlayFromStart)
      (:Update
        (Audio|Parameter|SetFloatParameter (Variables|Default|GetMusic) "Crossfade" _crossfade :!class /Script/Engine.AudioComponent)
        (Audio|Components|Audio|SetLowPassFilterFrequency (Variables|Default|GetMusic) (Math|Float|MapRangeClamped _crossfade 0.0 1.0 (Variables|Sound|GetLPFFrequencyMin) 2e+04) :!class /Script/Engine.AudioComponent)))))

(event Custom|Close%20Music%20Filter
  (params (Time :type Float%20%28double-precision%29))
  ;[1 1 1] Start / Stop Club Music
  (Audio|Components|Audio|SetLowPassFilterEnabled (Variables|Default|GetMusic) true :!class /Script/Engine.AudioComponent)
  (Components|Timeline|SetPlayRate (Variables|BP_Sneezy_FloorProp_Area|GetClub%20Music%20Fader) (/ 1.0 Time))
  (if (not (Variables|Sound|GetMusicIsFiltered))
    (Audio|PlaySoundatLocation (Variables|Sound|GetSoundOnMusicStop) (Transformation|GetWorldLocation (Variables|Default|GetMusic)))
    (Variables|Sound|SetMusicIsFiltered true)
    (Timeline|Club%20Music%20Fader ReverseFromEnd)))

(event Custom|Tech%20Line%20Updated
  (params (Health :type Float%20%28double-precision%29))
  (bind _tech_lines (Variables|Default|GetTech%20Lines))
  ;[0.08653 1 0.329] plz hook up to progress update event
  (Rendering|Material|SetScalarParameterValueonMaterials _tech_lines "GameTime" (Utilities|Time|GetGameTimeinSeconds))
  (Rendering|Material|SetScalarParameterValueonMaterials _tech_lines "Health" Health)
  (Rendering|Material|SetVectorParameterValueonMaterials _tech_lines "HitWave-Loc" (- (Transformation|GetWorldLocation (Variables|Default|GetGatewayHatch_Pedestal_pedestal02)) "0, 0,640.000000")))

(event Custom|LocalNeedleUp
  (bind (_direction _0_1) (Timeline|WobbleNeedle PlayFromStart)
    (:Update
      (Rendering|Material|SetScalarParameterValue (Variables|Default|GetSoundMeterMID) "NormalizedLevel" (Math|Float|Lerp (Variables|Default|GetMeterLevel) (+ (Variables|Default|GetMeterLevel) (Variables|Default|GetWobbleBias)) (select (<= (Variables|Default|GetSoundMeterStartingLevel) (Variables|Default|GetMeterLevel)) _0_1 (* _0_1 -1.0))) :!class /Script/Engine.MaterialInstanceDynamic))))

(event Custom|LocalNeedleDown
  (Timeline|WobbleNeedle PlayFromStart))

(fn OnRep_LocalEdgeDancers
  (Networking|SwitchHasAuthority
    (:Authority
      ;[1 1 1] Whenever the number of dancers changes (delta) send the result on
      (if (!= (Variables|Default|GetLocalEdgeDancers) 0)
        (Default|CallOnDancerCountChange (Variables|Default|GetLocalEdgeDancers) self)
        (Development|PrintString (Utilities|String|Append (Utilities|String|ToString(Integer) (Variables|Default|GetLocalEdgeDancers)) " local dancers") true true "(R=0.000000,G=0.660000,B=1.000000,A=1.000000)" 2e+01)))
    (:Remote
      (Utilities|FlowControl|Sequence
        (:then_0
          (if (== (Variables|Default|GetLocalEdgeDancers) (Variables|Default|GetPreviousDancerCount))
            _
            (elif (> (Variables|Default|GetLocalEdgeDancers) (Variables|Default|GetPreviousDancerCount))
              (CallFunction|LocalNeedleUp)
              (CallFunction|TechLineUpdated (Variables|Default|GetMeterLevel))
              (else
                (CallFunction|LocalNeedleDown)))))
        (:then_1
          ;[1 1 1] If anyone is locally dancing start the music!
          (if (> (Variables|Default|GetDancerCount) 0)
            (CallFunction|OpenMusicFilter 3.0)
            (else
              (CallFunction|CloseMusicFilter 3.0))))))))

(fn OnRep_MeterLevel
  (CallFunction|GlobalNeedleMove (Variables|Default|GetMeterLevel))
  ;[1 1 1] This is true if the server was reset and the global dancers was reset to 0 from the mesh
  (if (Math|Float|NearlyEqual(Float) (Variables|Default|GetMeterLevel))
    (Variables|Default|SetPreviousDancerCount 0)
    (if (not (Utilities|Time|IsValidTimerHandle (Variables|Default|GetDanceCheckTimer)))
      (bind _returnvalue (Utilities|Time|SetTimerbyEvent (EventDispatchers|CreateEvent :!event CheckForDancers) 1.0 true))
      (Variables|Default|SetDanceCheckTimer _returnvalue))))

(fn OnRep_PlayDancefloorMusic
  (Networking|CanExecuteCosmeticEvents
    (:True
      (if (Variables|Default|GetPlayDancefloorMusic)
        (Audio|Components|Audio|FadeIn (Variables|Default|GetAmb) 1.0 1.0 0.0 "Linear" :!class /Script/Engine.AudioComponent)
        (CallFunction|CloseMusicFilter 3.0)
        (else
          (Audio|Components|Audio|FadeOut (Variables|Default|GetAmb) 1.0 0.0 "Linear" :!class /Script/Engine.AudioComponent))))))

(fn ShouldDie
  (params (Damage :type Float%20%28single-precision%29) (EventInstigator :type Controller%20Object%20Reference) (DamageCauser :type Actor%20Object%20Reference) (bIsTest :type Boolean) (EffectContext :type Gameplay%20Effect%20Context%20Handle%20Structure%20%28by%20ref%29))
  (return false))

(fn UpdateMeterLevelAudio
  (params (Meter%20Level :type Float%20%28double-precision%29))
  (Networking|SwitchHasAuthority
    (:Remote
      (Audio|Parameter|SetFloatParameter (Variables|Default|GetMusic) "MeterLevel" Meter%20Level :!class /Script/Engine.AudioComponent))))

(fn ConstructionScript
  ;[5 3.613 0.1215] Set to false when checking in
  (Collision|SetActorEnableCollision false)
  (Networking|CanExecuteCosmeticEvents
    (:True
      (bind _returnvalue (Rendering|Material|CreateDynamicMaterialInstance (Variables|Default|GetSoundMeter) 0 0 "None" :!class /Script/Engine.PrimitiveComponent))
      (Variables|Default|SetSoundMeterMID _returnvalue)
      (bind _returnvalue_2 (Rendering|Material|CreateDynamicMaterialInstance (Variables|Default|GetFFT_DanceFloor) 0 0 "None" :!class /Script/Engine.PrimitiveComponent))
      (Variables|Default|SetFFT_DanceFloorMID _returnvalue_2)
      (Variables|Default|SetMagneticSplines (Utilities|Array|MakeArray (Variables|Default|GetMagneticSpline01) (Variables|Default|GetMagneticSpline02) (Variables|Default|GetMagneticSpline03) (Variables|Default|GetMagneticSpline04)))
      (Transformation|AttachComponentToComponent (Variables|Default|GetCoalesce) (Variables|Default|GetGatewayHatch_Pedestal_pedestal02) "FX_Attack" "KeepRelative" "KeepRelative" "KeepRelative" true)
      (for _array_element (Variables|Default|GetMagneticSplines)
        (Transformation|AttachComponentToComponent _array_element (Variables|Default|GetGatewayHatch_Pedestal_pedestal02) "FX_Attack" "KeepRelative" "KeepRelative" "KeepRelative" true)
        (bind _returnvalue_3 (Rendering|Material|CreateDynamicMaterialInstance _array_element 0 0 "None" :!class /Script/Engine.PrimitiveComponent))
        (Utilities|Array|Add (Variables|Default|GetMagneticSplineMIDs) _returnvalue_3)))))
